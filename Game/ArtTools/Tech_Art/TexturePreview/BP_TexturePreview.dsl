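; Art-tool actor that previews textures on a static or skeletal mesh through M_TexturePreview, with a normal-map flag.
; The construction script assigns the mesh, or shows a text render if neither type casts.
; It then fits Texture Slots to the material count and gives each slot a dynamic preview material.
; Optionally, newly grown slot arrays pull original textures by probing parameter-name lists for the chosen texture group, falling back to black.
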(event EventBeginPlay)

(event Collision|EventActorBeginOverlap
  (params (OtherActor :type Actor%20Object%20Reference)))

(event EventTick
  (params (DeltaSeconds :type Float%20%28single-precision%29)))

(fn SetMaterials
  (params (Mesh :type Object%20Reference))
  (bind _asstatic_mesh_component (Utilities|Casting|CastToStaticMeshComponent Mesh)
    (:then
      (Variables|Default|SetPrimitive _asstatic_mesh_component))
    (:CastFailed
      (bind _asskeletal_mesh_component (Utilities|Casting|CastToSkeletalMeshComponent Mesh))
      (Variables|Default|SetPrimitive _asskeletal_mesh_component)))
  (for (_array_element _array_index) (Variables|Default|GetTexture%20Slots)
    (bind _returnvalue (Rendering|Material|CreateDynamicMaterialInstance "/Game/ArtTools/Tech_Art/TexturePreview/M_TexturePreview.M_TexturePreview" "None" "None" :!class /Script/Engine.KismetMaterialLibrary))
    (Rendering|Material|SetTextureParameterValue _returnvalue "Preview" _array_element)
    (Rendering|Material|SetScalarParameterValue _returnvalue "IsNormal" (Math|Conversions|ToFloat(Boolean) (Variables|Default|GetIsNormalMap)) :!class /Script/Engine.MaterialInstanceDynamic)
    (Rendering|Material|SetMaterial (Variables|Default|GetPrimitive) _array_index _returnvalue)))

(fn SetTextureArrayLength
  (params (Mesh%20Component :type Primitive%20Component%20Object%20Reference))
  (LocalVariables|SetLocal%20Mesh%20Component Mesh%20Component)
  (if (and (> (Utilities|Array|Length (Variables|Default|GetTexture%20Slots)) (Rendering|Material|GetNumMaterials (LocalVariables|GetLocal%20Mesh%20Component))) (not (Variables|Default|GetPrevent%20Texture%20Deletion)))
    (for _ (range (+ (- (Utilities|Array|Length (Variables|Default|GetTexture%20Slots)) (Rendering|Material|GetNumMaterials (LocalVariables|GetLocal%20Mesh%20Component))) 1))
      (Utilities|Array|RemoveIndex (Variables|Default|GetTexture%20Slots) (Utilities|Array|LastIndex (Variables|Default|GetTexture%20Slots))))
    (elif (< (Utilities|Array|Length (Variables|Default|GetTexture%20Slots)) (Rendering|Material|GetNumMaterials (LocalVariables|GetLocal%20Mesh%20Component)))
      (for _ (range (+ (- (Rendering|Material|GetNumMaterials (LocalVariables|GetLocal%20Mesh%20Component)) (Utilities|Array|Length (Variables|Default|GetTexture%20Slots))) 1))
        (Utilities|Array|Add (Variables|Default|GetTexture%20Slots)))
      (if (Variables|Default|GetGet%20Original%20Texture)
        (for (_ _array_index) (Variables|Default|GetTexture%20Slots)
          (bind _returnvalue (Rendering|Material|CreateDynamicMaterialInstance (Rendering|Material|GetMaterial (LocalVariables|GetLocal%20Mesh%20Component) _array_index) "None" "None" :!class /Script/Engine.KismetMaterialLibrary))
          (bind _texture_asset (CallFunction|TryGetTexture _returnvalue))
          (Utilities|Array|SetArrayElem (Variables|Default|GetTexture%20Slots) _array_index _texture_asset)))
      (else
        _))))

(fn TryGetOriginalTextures
  (for (_ _array_index) (Variables|Default|GetTexture%20Slots)
    (bind _returnvalue (Rendering|Material|CreateDynamicMaterialInstance (Rendering|Material|GetMaterial (Variables|Default|GetPrimitive) _array_index) "None" "None" :!class /Script/Engine.KismetMaterialLibrary))
    (bind _texture_asset (CallFunction|TryGetTexture _returnvalue))
    (bind _returnvalue_2 (Rendering|Material|CreateDynamicMaterialInstance (Variables|Default|GetPrimitive) _array_index "/Game/ArtTools/Tech_Art/TexturePreview/M_TexturePreview.M_TexturePreview" "None" :!class /Script/Engine.PrimitiveComponent))
    (Rendering|Material|SetTextureParameterValue _returnvalue_2 "Preview" _texture_asset)))

(fn TryGetTexture
  (params (Material :type Material%20Instance%20Dynamic%20Object%20Reference))
  (bind _texture_group (Variables|Default|GetTexture%20Group))
  (bind _misc3list (Variables|Default|GetMisc3List))
  (bind _misc2list (Variables|Default|GetMisc2List))
  (bind _misclist (Variables|Default|GetMiscList))
  (bind _maskslist (Variables|Default|GetMasksList))
  (bind _mlist (Variables|Default|GetMList))
  (bind _specularlist (Variables|Default|GetSpecularList))
  (bind _normallist (Variables|Default|GetNormalList))
  (bind _emissivelist (Variables|Default|GetEmissiveList))
  (bind _diffuselist (Variables|Default|GetDiffuseList))
  (bind _returnvalue (Utilities|Select :Index _texture_group :NewEnumerator0 _diffuselist :NewEnumerator1 _emissivelist :NewEnumerator2 _normallist :NewEnumerator3 _specularlist :NewEnumerator4 _mlist :NewEnumerator5 _maskslist :NewEnumerator6 _misclist :NewEnumerator7 _misc2list :NewEnumerator9 _misc3list))
  (bind _index (Utilities|FlowControl|ForLoopwithBreak 0 (Utilities|Array|LastIndex _returnvalue))
    (:LoopBody
      (bind _returnvalue_2 (Rendering|Material|GetTextureParameterValue Material (Utilities|Array|Get(aref) _returnvalue _index) :!class /Script/Engine.MaterialInstanceDynamic))
      (Utilities|IsValid _returnvalue_2
        (:Is%20Valid
          (LocalVariables|SetTexture _returnvalue_2))))
    (:Completed
      _))
  (Utilities|IsValid (LocalVariables|GetTexture)
    (:Is%20Valid
      (LocalVariables|SetbTextureFound true))
    (:Is%20Not%20Valid
      (LocalVariables|SetbTextureFound false)
      (LocalVariables|SetTexture "/Game/Global/Textures/Default/Blanks/T_BlacksRGB.T_BlacksRGB")))
  (return (LocalVariables|GetTexture) (LocalVariables|GetbTextureFound)))

(fn ConstructionScript
  (bind _skeletal_mesh (Variables|Default|GetSkeletal%20Mesh))
  (bind _static_mesh (Variables|Default|GetStatic%20Mesh))
  (bind _asstatic_mesh (Utilities|Casting|CastToStaticMesh (Variables|Default|GetMesh))
    (:then
      (bind _output_get (Variables|Default|SetStatic%20Mesh (Variables|Default|GetSM)))
      (Components|StaticMesh|SetStaticMesh _output_get _asstatic_mesh)
      (CallFunction|SetTextureArrayLength _static_mesh)
      (CallFunction|SetMaterials _static_mesh))
    (:CastFailed
      (bind _asskeletal_mesh (Utilities|Casting|CastToSkeletalMesh (Variables|Default|GetMesh))
        (:then
          (bind _output_get_2 (Variables|Default|SetSkeletal%20Mesh (Variables|Default|GetSK)))
          (Components|SkeletalMesh|SetSkeletalMeshAsset _output_get_2 _asskeletal_mesh)
          (CallFunction|SetTextureArrayLength _skeletal_mesh)
          (CallFunction|SetMaterials _skeletal_mesh))
        (:CastFailed
          (Rendering|SetVisibility (Variables|Default|GetTextRender) true false))))))
