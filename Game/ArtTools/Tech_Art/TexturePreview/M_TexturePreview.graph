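# M_TexturePreview is an unlit tech-art material for previewing a texture, flagged for use on skeletal, clothing and mesh-deformer meshes.
# The Preview texture's RGB feeds a lerp against a vector-transformed copy of itself, blended by the isNormal scalar so normal maps display differently.
# A separate PreviewNormal sampler with a flat normal default feeds nothing, and the graph shows no link from the lerp to a material output.
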Material /Game/ArtTools/Tech_Art/TexturePreview/M_TexturePreview
graph {
 properties {
  bUsedWithClothing=true
  bUsedWithMeshDeformer=true
  bUsedWithSkeletalMesh=true
  pixelDepthOffsetMode=PDOM_Legacy
  shadingModel=MSM_Unlit
  usedShadingModels=Unlit
 }
 nodes [
  n1 = TextureSampleParameter2D() {parameterName=Preview, texture=&T_BlacksRGB}
  LinearInterpolate(A=@n1.RGB, B=Transform(None=@n1.RGB), Alpha=ScalarParameter() {parameterName=isNormal})
  TextureSampleParameter2D() {parameterName=PreviewNormal, samplerType=SAMPLERTYPE_Normal, texture=&FlatNormalBlue}
 ]
}
